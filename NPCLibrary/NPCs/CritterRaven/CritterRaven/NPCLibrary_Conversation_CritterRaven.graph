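# Conversation database for the CritterRaven NPC in the NPC library, defining its sand-themed greeting and the services it offers.
# The entry tag leads to one speech node with a default line and contextual lines for first meeting, time since last match, visiting Sunny Shores, and holding Bars.
# Each contextual line takes its priority from a shared curve table row.
# The speech then branches through link nodes into the shared NPC quest and basic services conversations.
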ConversationDatabase /NPCLibrary/NPCs/CritterRaven/CritterRaven/NPCLibrary_Conversation_CritterRaven
entry_points [{tag=Conversation.Entry.CritterRaven, destinations [#1]}]
nodes [
 ConversationEntryPointNode_7 {
  id=#1
  class=ConversationEntryPointNode
  outputs [#2]
  properties {entryTag {tagName=Conversation.Entry.CritterRaven}}
 }
 Speech_C_4 {
  id=#2
  class=Speech_C
  outputs [#3, #4]
  properties {
   nodeName=""
   generalConfig {
    defaultMessage=You've travelled across the sands to see me? Speak.
    contextualMessages [
     {
      message=You've returned... Just as the sands of time have begun shifting under our feet...
      contextRequirements [{requirement {class=FortControllerRequirement_CheckTimeSinceLastPlayedMatch}}]
      priority {curve {rowName=Tandem.Speech.Priority.WhileSincePlayed}}
     }
     {
      message=How were the sandy Beaches of Believer?
      contextRequirements [{requirement {class=FortControllerRequirement_HasVisitedPOI, properties {pOITag {tagName=Athena.Location.POI.SunnyShores}}}}]
      priority {curve {rowName=Tandem.Speech.Priority.VisitedSpecificPOI}}
     }
     {
      message="Fear the master of the dunes. At minimum, I will make you itchy."
      contextRequirements [{requirement {class=FortControllerRequirement_IsFirstEverConversationWithNPC}}]
      priority {curve {rowName=Tandem.Speech.Priority.FirstEverInteraction}}
     }
     {
      message=So many Bars... like grains of sand running through your fingers...
      contextRequirements [{requirement {class=FortControllerRequirement_HasBars}}]
      priority {curve {rowName=Tandem.Speech.Priority.HighTotalBars}}
     }
    ]
   }
  }
 }
 ConversationLinkNode_0 {id=#4, class=ConversationLinkNode, properties {remoteEntryTag {tagName=Conversation.Entry.BasicServices}}}
 ConversationLinkNode_1 {id=#3, class=ConversationLinkNode, properties {remoteEntryTag {tagName=Conversation.Entry.NPCLibrary.Quests}}}
]
